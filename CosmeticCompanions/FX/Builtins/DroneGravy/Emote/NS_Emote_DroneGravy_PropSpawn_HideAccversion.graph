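# Niagara system for the DroneGravy cosmetic companion's emote, named as a prop spawn/hide effect variant.
# It holds two sets of sparkle, airglow and smoke emitters, apparently lightweight given the LWE suffix, with only SparklesLWE001 and AirglowLWE001 enabled.
# A User.NormalizedNotifyProgress float is exposed, though the file shows no emitter modules, renderers or bindings that use it.
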NiagaraSystem /CosmeticCompanions/FX/Builtins/DroneGravy/Emote/NS_Emote_DroneGravy_PropSpawn_HideAccversion
dependencies {usedDataInterfaces [], usedDynamicInputs [], usedModules [&SystemState]}
emitters {
 AirglowLWE {bEnabled=false, rendererClasses [], renderers []}
 AirglowLWE001 {rendererClasses [], renderers []}
 SmokeLWE {bEnabled=false, rendererClasses [], renderers []}
 SmokeLWE001 {bEnabled=false, rendererClasses [], renderers []}
 SparklesLWE {bEnabled=false, rendererClasses [], renderers []}
 SparklesLWE001 {rendererClasses [], renderers []}
}
summary {
 systemName=NS_Emote_DroneGravy_PropSpawn_HideAccversion
 emitters [
  {bEnabled=false, emitterName=SparklesLWE, rendererClasses []}
  {bEnabled=false, emitterName=AirglowLWE, rendererClasses []}
  {bEnabled=false, emitterName=SmokeLWE, rendererClasses []}
  {emitterName=SparklesLWE001, rendererClasses []}
  {emitterName=AirglowLWE001, rendererClasses []}
  {bEnabled=false, emitterName=SmokeLWE001, rendererClasses []}
 ]
 userVariables [User.NormalizedNotifyProgress {type=NiagaraFloat, defaultValue=0, description=""}]
}
